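# Legacy horizontal tab bar widget on FortTabListWidgetBase_Legacy, with a scrolling button row, previous/next input prompts and optional arrow buttons that broadcast dispatchers.
# New tabs pass through UpdateTabStyles for parenting, padding, sizing, style and optional label/icon suppression; selection can scroll them into view.
# Visible-tab count drives prompt visibility, while keybind changes toggle arrows by input type.
# PreConstruct binds events and builds placeholder tabs for designer preview.
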WidgetBlueprint /Game/UI/Foundation/TabbedView/HorizontalTabList_Legacy
graphs [
 DebugCreateTabsForDesigner {nodes [
   DebugCreateTabsForDesigner() {is=title K2Node_FunctionEntry}
   > ClearChildren(self=Get TabButtonScrollBox() -> (TabButtonScrollBox:Scroll Box Object Reference))
   > For Loop(FirstIndex=1, LastIndex=Get DebugTabCount() -> (DebugTabCount:int=0))
   > n1 = Create Widget!(
    execute:Exec
    Class:User Widget Class Reference=Get DebugButtonTypeDisplay() -> (DebugButtonTypeDisplay:Common Button Base Class Reference)
    OwningPlayer:Player Controller Object Reference=GetOwningPlayer()
    Style:Common Button Style Class Reference
    bApplyAlphaOnDisable:bool=True
    bLocked:bool=False
    bSelectable:bool=False
    bShouldSelectUponReceivingFocus:bool=False
    bInteractableWhenSelected:bool=False
    bToggleable:bool=False
    bTriggerClickedAfterSelection:bool=False
    bDisplayInputActionWhenNotInteractable:bool=True
    bHideInputActionWithKeyboard:bool=False
    bShouldUseFallbackDefaultInputAction:bool=True
    bRequiresHold:bool=False
    HoldData:Common UIHold Data Class Reference
    ClickMethod:EButtonClickMethod Enum=DownAndUp
    InputPriority:int=0
    TriggeringInputAction:Data Table Row Handle Structure
   ) -> (then:Exec, ReturnValue:Common Button Base Object Reference)
   > UpdateTabStyles!(execute, self, ButtonWidget:Common Button Base Object Reference=@n1) -> (then)
  ]}
 EventGraph {nodes [
   "On Clicked (ButtonLeft)"!() -> (OutputDelegate, then, Button:Common Button Base Object Reference) {is=title K2Node_ComponentBoundEvent}
   > Call On Previous Tab!(execute:Exec, self:Horizontal Tab List Legacy Object Reference) -> (then:Exec) {is=title K2Node_CallDelegate}
   "On Clicked (ButtonRight)"!() -> (OutputDelegate, then, Button:Common Button Base Object Reference) {is=title K2Node_ComponentBoundEvent}
   > Call On Next Tab!(execute:Exec, self:Horizontal Tab List Legacy Object Reference) -> (then:Exec) {is=title K2Node_CallDelegate}
   n1 = CustomEvent_1!() -> (OutputDelegate, then, InVisibility:ESlateVisibility Enum=Visible) {is=title K2Node_CustomEvent}
   > Refresh Next/Previous Visibility()
   n2 = CustomEvent!() -> (OutputDelegate, then, TabId:Name) {is=title K2Node_CustomEvent}
   > Branch(Condition=Get bUseScroll() -> (bUseScroll:bool))
   > n3 = GetTabButtonByID!(execute, self:Common Tab List Widget Legacy Object Reference, TabNameID:Name=@n2.TabId) -> (then, ReturnValue:Common Button Legacy Object Reference)
   > ScrollWidgetIntoView!(
    execute
    self:Scroll Box Object Reference=Get TabButtonScrollBox()
    WidgetToFind:Widget Object Reference=@n3
    AnimateScroll:bool=true
    ScrollDestination:EDescendantScrollDestination Enum=Center
    Padding:float=15
   ) -> (then)
   Refresh Next/Previous Visibility() {is=title K2Node_CustomEvent}
   n4 = GetNumVisibleTabs!(self) -> (Num:int=0)
   > SetNextPrevTabActionWrappersVisibility!(
    execute
    self
    InVisibility:ESlateVisibility Enum=Select(Option 0=Visible, Option 1=Hidden, Index="integer <= integer"!(A:int=@n4, B:int=1) -> (ReturnValue:bool) {is=title K2Node_CallFunction})
   ) -> (then)
   n5 = OnVisibilityChanged_Event!() -> (OutputDelegate, then, InVisibility:ESlateVisibility Enum=Visible) {is=title K2Node_CustomEvent}
   > Refresh Next/Previous Visibility() {is=title K2Node_CallFunction}
   n6 = CustomEvent_0!() -> (OutputDelegate, then, TabId:Name) {is=title K2Node_CustomEvent}
   > Refresh Next/Previous Visibility()
   > Call On Tab Changed!(execute:Exec, self:Horizontal Tab List Legacy Object Reference) -> (then:Exec) {is=title K2Node_CallDelegate}
   n7 = CustomEvent_2() {is=title K2Node_CustomEvent}
   n8 = Get CommonInputSubsystem!() -> (ReturnValue:Common Input Object Reference) {is=title K2Node_GetBlueprintContextEx}
   > Is Valid(InputObject=@n8)
   > Branch(Condition=Get bUseScrollButtons() -> (bUseScrollButtons:bool))
   n9 = GetCurrentInputType(self=@n8)
   > n10 = Branch(Condition="Equal (Enum)"(A=@n9, B=MouseAndKeyboard))
   > SetVisibility(
    self=[Get ButtonLeft() -> (ButtonLeft:Icon Text Button Object Reference), Get ButtonRight() -> (ButtonRight:Icon Text Button Object Reference)]
    InVisibility=SelfHitTestInvisible
   )
   Branch(execute=@n10.else, Condition="Equal (Enum)"(A=@n9, B=Gamepad))
   > SetVisibility(self=[Get ButtonLeft(), Get ButtonRight()], InVisibility=Collapsed)
   n11 = Event HandleTabCreation!() -> (OutputDelegate, then, TabNameID:Name=None, TabButton:Common Button Base Object Reference) {is=title K2Node_Event}
   > Parent: HandleTabCreation!(execute, TabNameID:Name=@n11.TabNameID, TabButton:Common Button Base Object Reference=@n11.TabButton) -> (then) {is=title K2Node_CallParentFunction}
   > UpdateTabStyles(ButtonWidget=@n11.TabButton)
   > Assign On Visibility Changed!(execute:Exec, self:User Widget Object Reference=@n11.TabButton, Delegate:"Delegate (by ref)"=@n5) -> (then:Exec) {is=title K2Node_AssignDelegate}
   n12 = Event HandleTabRemoval!() -> (OutputDelegate, then, TabNameID:Name=None, TabButton:Common Button Base Object Reference) {is=title K2Node_Event}
   > Parent: HandleTabRemoval!(execute, TabNameID:Name=@n12.TabNameID, TabButton:Common Button Base Object Reference=@n12.TabButton) -> (then) {is=title K2Node_CallParentFunction}
   > Unbind Event from On Visibility Changed(self:User Widget Object Reference=@n12.TabButton, Delegate=@n5) {is=title K2Node_RemoveDelegate}
   > Refresh Next/Previous Visibility()
   n13 = Event PreConstruct()
   > Parent: PreConstruct(IsDesignTime=@n13.IsDesignTime)
   > n14 = Sequence()
   > SetInputAction!(
    execute
    self:Common Action Widget Object Reference=Get PreviousTabAction() -> (PreviousTabAction:Common Action Widget Object Reference)
    InputActionRow:Data Table Row Handle Structure=Get PreviousTabInputActionData() -> (PreviousTabInputActionData:Data Table Row Handle Structure)
   ) -> (then)
   > SetInputAction(
    self=Get NextTabAction() -> (NextTabAction:Common Action Widget Object Reference)
    InputActionRow=Get NextTabInputActionData() -> (NextTabInputActionData:Data Table Row Handle Structure)
   )
   > Assign On Tab Selected!(execute:Exec, self:Horizontal Tab List Legacy Object Reference, Delegate:"Delegate (by ref)"=@n6) -> (then:Exec) {is=title K2Node_AssignDelegate}
   > Bind Event to On Visibility Changed(self:User Widget Object Reference, Delegate=@n1) {is=title K2Node_AddDelegate}
   > Bind Event to On Tab Selected(self:Common Tab List Widget Base Object Reference, Delegate=@n2) {is=title K2Node_AddDelegate}
   n15 = Get FortGlobalUIContext!() -> (ReturnValue:Fort Global UIContext Object Reference) {is=title K2Node_GetBlueprintContextEx}
   > Is Valid(InputObject=@n15)
   > Bind Event to On Keybinds Changed(self:Fort Global UIContext Object Reference=@n15, Delegate=@n7) {is=title K2Node_AddDelegate}
   > n16 = Branch(Condition=Get bUseScrollButtons())
   > n17 = SetVisibility(self=[Get ButtonLeft(), Get ButtonRight()], InVisibility=SelfHitTestInvisible)
   n18 = Branch(execute=[@n17.then, @n20.then], Condition=Get bUseSizeOverride() -> (bUseSizeOverride:bool))
   n19 = Get TabButtonWrapper() -> (TabButtonWrapper:Size Box Object Reference)
   > SetWidthOverride(self=@n19, InWidthOverride=600)
   ClearWidthOverride(execute=@n18.else, self:Size Box Object Reference=@n19)
   n20 = SetVisibility(execute=@n16.else, self=[Get ButtonLeft(), Get ButtonRight()], InVisibility=Collapsed)
   n21 = Get DebugButtonTypeDisplay()
   Branch(execute=@n14.then_1, Condition=AND Boolean(A=@n13.IsDesignTime, B=IsValidClass!(Class:Object Class Reference=@n21) -> (ReturnValue:bool)))
   > DebugCreateTabsForDesigner() {is=title K2Node_CallFunction}
   Event Construct()
   > Refresh Next/Previous Visibility()
   > SetVisibility(
    self=Get ImageHorizRule() -> (ImageHorizRule:Image Object Reference)
    InVisibility=Select(Option 0=Collapsed, Option 1=SelfHitTestInvisible, Index=Get ShowRail() -> (ShowRail:bool))
   )
  ]}
 GetNumVisibleTabs {nodes [
   GetNumVisibleTabs() {is=title K2Node_FunctionEntry}
   > n1 = For Loop(LastIndex=int - int(A=GetTabCount!(self:Common Tab List Widget Base Object Reference) -> (ReturnValue:int=0)))
   Return Node!(execute=@n1.Completed, Num:int=Get NumVisibleTabs() -> (NumVisibleTabs:int=0))
   n2 = GetTabIdAtIndex!(self:Common Tab List Widget Base Object Reference, Index:int=@n1) -> (ReturnValue:Name=None)
   > Branch(Condition=IsVisible(self=GetTabButtonBaseByID!(self:Common Tab List Widget Base Object Reference, TabNameID:Name=@n2) -> (ReturnValue:Common Button Base Object Reference)))
   > Increment Int(Value=Get NumVisibleTabs() -> (NumVisibleTabs=0))
  ]}
 OnNextTab {nodes [OnNextTab() {is=title K2Node_FunctionEntry}]}
 OnPreviousTab {nodes [OnPreviousTab() {is=title K2Node_FunctionEntry}]}
 OnTabChanged {nodes [OnTabChanged() {is=title K2Node_FunctionEntry}]}
 RefreshGamepad {nodes [RefreshGamepad() {is=title K2Node_FunctionEntry}]}
 SetNextPrevTabActionWrappersVisibility {nodes [
   n1 = SetNextPrevTabActionWrappersVisibility!() -> (then, InVisibility:ESlateVisibility Enum=Visible) {is=title K2Node_FunctionEntry}
   > SetVisibility(self=Get PreviousTabActionWrapper() -> (PreviousTabActionWrapper:Size Box Object Reference), InVisibility=@n1)
   > SetVisibility(self=Get NextTabActionWrapper() -> (NextTabActionWrapper:Size Box Object Reference), InVisibility=@n1)
  ]}
 SetVerticalAlignment {nodes [
   n1 = SetVerticalAlignment!() -> (then, FillVertically:EVerticalAlignment Enum=VAlign_Fill) {is=title K2Node_FunctionEntry}
   > SetVerticalAlignment(
    self:Horizontal Box Slot Object Reference=SlotAsHorizontalBoxSlot!(Widget:Widget Object Reference=Get Overlay_Container() -> (Overlay_Container:Overlay Object Reference)) -> (ReturnValue:Horizontal Box Slot Object Reference)
    InVerticalAlignment=@n1
   ) {is=title K2Node_CallFunction}
  ]}
 UpdateTabStyles {nodes [
   n1 = UpdateTabStyles!() -> (then, ButtonWidget:Common Button Base Object Reference) {is=title K2Node_FunctionEntry}
   > n2 = Branch(Condition=Get ForceTabLabelsCollapsed() -> (ForceTabLabelsCollapsed:bool))
   > n3 = Cast To IconTabButton_Legacy!(execute, Object=@n1) -> (then, CastFailed, AsIcon Tab Button Legacy:Icon Tab Button Legacy Object Reference) {is=title K2Node_DynamicCast}
   > n4 = Set ButtonText!(execute, ButtonText:Text=GetEmptyText!() -> (ReturnValue:Text), self:Icon Tab Button Legacy Object Reference=@n3) -> (then, Output_Get:Text)
   n5 = Branch(execute=[@n4.then, @n3.CastFailed, @n2.else], Condition=Get ForceTabIconsCollapsed() -> (ForceTabIconsCollapsed:bool))
   > n6 = Cast To IconTabButton_Legacy(Object=@n1)
   > n7 = SetVisibility(self=Get LeftSideImage!(self:Icon Tab Button Legacy Object Reference=@n6) -> (LeftSideImage:Image Object Reference), InVisibility=Collapsed)
   AddChild(execute=[@n7.then, @n6.CastFailed, @n5.else], self=Get TabButtonScrollBox(), Content=@n1)
   > SetPadding!(execute, self:User Widget Object Reference=@n1, InPadding:Margin Structure=Get TabButtonPadding() -> (TabButtonPadding:Margin Structure)) -> (then)
   n8 = SlotAsSizeBoxSlot!(Widget:Widget Object Reference=Get NextTabAction()) -> (ReturnValue:Size Box Slot Object Reference)
   n9 = Break Margin!(Margin:"Margin Structure (by ref)"=Get Padding!(self:Size Box Slot Object Reference=@n8) -> (Padding:Margin Structure)) -> (Left:float=~, Top:float=~, Right:float=~, Bottom:float=~) {is=title K2Node_BreakStruct}
   > SetPadding(
    self:Size Box Slot Object Reference=@n8
    InPadding=Make Margin(Left=Get NextTabActionPaddingLeft() -> (NextTabActionPaddingLeft:double), Top=@n9.Top, Right=@n9.Right, Bottom=@n9.Bottom)
   )
   n10 = SlotAsSizeBoxSlot(Widget=Get PreviousTabAction())
   n11 = Break Margin(Margin=Get Padding(self=@n10)) -> (Left=~, Top=~, Right=~, Bottom=~)
   > SetPadding(
    self=@n10
    InPadding=Make Margin(Left=@n11, Top=@n11.Top, Right=Get PreviousTabActionPaddingRight() -> (PreviousTabActionPaddingRight:double), Bottom=@n11.Bottom)
   )
   > SetMinDimensions!(execute, self:Common Button Base Object Reference=@n1, InMinWidth:int=Get MinTabWidth() -> (MinTabWidth:int=0), InMinHeight:int=0) -> (then)
   > Refresh Next/Previous Visibility()
   > Branch(Condition=Get UseButtonStyle() -> (UseButtonStyle:bool))
   > SetStyle!(
    execute
    self:Common Button Base Object Reference=@n1
    InStyle:Common Button Style Class Reference=Get ButtonStyle() -> (ButtonStyle:Common Button Style Class Reference)
   ) -> (then)
  ]}
]
variables [
 bUseScroll {type=boolean}
 bUseScrollButtons {type=boolean}
 bUseSizeOverride {type=boolean}
 ButtonStyle {type=Class, default=null}
 DebugButtonTypeDisplay {type=Class, default=null}
 DebugTabCount {type=integer, default=3}
 ForceTabIconsCollapsed {type=boolean}
 ForceTabLabelsCollapsed {type=boolean}
 MinTabWidth {type=integer}
 NextTabActionPaddingLeft {type=number}
 OnNextTab {}
 OnPreviousTab {}
 OnTabChanged {}
 PreviousTabActionPaddingRight {type=number}
 RefreshGamepad {}
 ShowGamePadInputPrompt {type=boolean}
 ShowRail {type=boolean}
 TabButtonPadding {type=Margin, default=<ltrb 0 0 0 0>}
 UseButtonStyle {type=boolean}
]
widgets {
 info {inheritedWidgetCount=0, parentClass=/Script/FortniteUI.FortTabListWidgetBase_Legacy, rootWidgetClass=&SizeBox, widgetCount=15}
 widgets [
  SizeBox_231 {widgetClassPath=&SizeBox, widget=SizeBox_231, properties {heightOverride=88}}
  HorizontalBox_1 {widgetClassPath=&HorizontalBox, parent=SizeBox_231, widget=HorizontalBox_1, slot {class=SizeBoxSlot}}
  Overlay_Container {
   widgetClassPath=&Overlay
   bIsVariable=true
   parent=HorizontalBox_1
   widget=Overlay_Container
   slot {class=HorizontalBoxSlot, properties {verticalAlignment=VAlign_Center}}
  }
  HorizontalBox_0 {
   widgetClassPath=&HorizontalBox
   parent=Overlay_Container
   widget=HorizontalBox_0
   slot {class=OverlaySlot, properties {horizontalAlignment=HAlign_Fill, verticalAlignment=VAlign_Fill}}
  }
  Overlay_0 {
   widgetClassPath=&Overlay
   parent=HorizontalBox_0
   widget=Overlay_0
   slot {class=HorizontalBoxSlot, properties {horizontalAlignment=HAlign_Left, verticalAlignment=VAlign_Center}}
  }
  PreviousTabActionWrapper {widgetClassPath=&SizeBox, bIsVariable=true, parent=Overlay_0, widget=PreviousTabActionWrapper, slot {class=OverlaySlot}}
  PreviousTabAction {
   widgetClassPath=&CommonActionWidget
   bIsVariable=true
   parent=PreviousTabActionWrapper
   widget=PreviousTabAction
   slot {class=SizeBoxSlot}
  }
  ButtonLeft {
   widgetClassPath=&IconTextButton
   bIsVariable=true
   parent=Overlay_0
   widget=ButtonLeft
   properties {
    bIsFocusable=false
    buttonText=""
    clickMethod=MouseDown
    onButtonBaseClicked=()
    onButtonBaseDoubleClicked=()
    onButtonBaseHovered=()
    onButtonBaseUnhovered=()
    onSelectedChangedBase=()
    style=/FortUILibrary/Foundation/Buttons/ButtonStyles/ButtonStyle-Rotator-Small.ButtonStyle-Rotator-Small_C
    visibility=Collapsed
    iconBrush {
     bIsDynamicallyLoaded=false
     drawAs=Image
     imageType=NoImage
     mirroring=NoMirror
     resourceName=None
     resourceObject=/FortUILibrary/Foundation/Buttons/Materials/M-Button-Outline-Base-Left
     tiling=NoTile
     imageSize=<12 12>
     margin=<ltrb 0 0 0 0>
     outlineSettings {
      bUseBrushTransparency=false
      roundingType=HalfHeightRadius
      width=0
      color {colorUseRule=UseColor_Specified, specifiedColor=<rgba 0 0 0 0>}
      cornerRadii=<0 0 0 1>
     }
     tintColor {colorUseRule=UseColor_Specified, specifiedColor=<rgba 0 0 0 1>}
     uVRegion {bIsValid=false, max=<0 0>, min=<0 0>}
    }
    pressedSlateSoundOverride {resourceObject=/Game/Sounds/Fort_UI_Sounds/FrontEnd/New_FrontEnd/DefaultSounds/PressedButtons/FN_UI_Default_Button_Pressed_Cue}
   }
   slot {class=OverlaySlot, properties {padding=<ltrb 0 0 25 0>}}
  }
  TabButtonWrapper {
   widgetClassPath=&SizeBox
   bIsVariable=true
   parent=HorizontalBox_0
   widget=TabButtonWrapper
   slot {class=HorizontalBoxSlot, properties {horizontalAlignment=HAlign_Center}}
  }
  TabButtonScrollBox {
   widgetClassPath=&ScrollBox
   bIsVariable=true
   parent=TabButtonWrapper
   widget=TabButtonScrollBox
   properties {
    navigationDestination=Center
    navigationScrollPadding=15
    orientation=Orient_Horizontal
    scrollBarVisibility=Collapsed
    scrollbarPadding=<ltrb 0 0 0 0>
    scrollbarThickness=<0 0>
    widgetStyle {
     barThickness=9
     bottomShadowBrush {
      bIsDynamicallyLoaded=false
      drawAs=NoDrawType
      imageType=FullColor
      mirroring=NoMirror
      resourceName=../../../Engine/Content/Slate/Common/ScrollBoxShadowBottom.png
      resourceObject=null
      tiling=NoTile
      imageSize=<16 8>
      margin=<ltrb 0.5 0 0.5 1>
      outlineSettings {
       bUseBrushTransparency=false
       roundingType=HalfHeightRadius
       width=0
       color {colorUseRule=UseColor_Specified, specifiedColor=<rgba 0 0 0 0>}
       cornerRadii=<0 0 0 0>
      }
      tintColor {colorUseRule=UseColor_Specified, specifiedColor=<rgba 1 1 1 1>}
      uVRegion {bIsValid=false, max=<0 0>, min=<0 0>}
     }
     horizontalScrolledContentPadding=<ltrb 0 0 0 0>
     leftShadowBrush {
      bIsDynamicallyLoaded=false
      drawAs=Image
      imageType=FullColor
      mirroring=NoMirror
      resourceName=../../../Engine/Content/Slate/Common/ScrollBoxShadowLeft.png
      resourceObject=null
      tiling=NoTile
      imageSize=<8 16>
      margin=<ltrb 1 0.5 0 0.5>
      outlineSettings {
       bUseBrushTransparency=false
       roundingType=HalfHeightRadius
       width=0
       color {colorUseRule=UseColor_Specified, specifiedColor=<rgba 0 0 0 0>}
       cornerRadii=<0 0 0 0>
      }
      tintColor {colorUseRule=UseColor_Specified, specifiedColor=<rgba 0.0176 0.0232 0.0331 0.65>}
      uVRegion {bIsValid=false, max=<0 0>, min=<0 0>}
     }
     rightShadowBrush {
      bIsDynamicallyLoaded=false
      drawAs=Image
      imageType=FullColor
      mirroring=NoMirror
      resourceName=../../../Engine/Content/Slate/Common/ScrollBoxShadowRight.png
      resourceObject=null
      tiling=NoTile
      imageSize=<8 16>
      margin=<ltrb 0 0.5 1 0.5>
      outlineSettings {
       bUseBrushTransparency=false
       roundingType=HalfHeightRadius
       width=0
       color {colorUseRule=UseColor_Specified, specifiedColor=<rgba 0 0 0 0>}
       cornerRadii=<0 0 0 0>
      }
      tintColor {colorUseRule=UseColor_Specified, specifiedColor=<rgba 0.0176 0.0232 0.0331 0.65>}
      uVRegion {bIsValid=false, max=<0 0>, min=<0 0>}
     }
     topShadowBrush {
      bIsDynamicallyLoaded=false
      drawAs=NoDrawType
      imageType=FullColor
      mirroring=NoMirror
      resourceName=../../../Engine/Content/Slate/Common/ScrollBoxShadowTop.png
      resourceObject=null
      tiling=NoTile
      imageSize=<16 8>
      margin=<ltrb 0.5 1 0.5 0>
      outlineSettings {
       bUseBrushTransparency=false
       roundingType=HalfHeightRadius
       width=0
       color {colorUseRule=UseColor_Specified, specifiedColor=<rgba 0 0 0 0>}
       cornerRadii=<0 0 0 0>
      }
      tintColor {colorUseRule=UseColor_Specified, specifiedColor=<rgba 1 1 1 1>}
      uVRegion {bIsValid=false, max=<0 0>, min=<0 0>}
     }
     verticalScrolledContentPadding=<ltrb 0 0 0 0>
    }
   }
   slot {class=SizeBoxSlot, properties {horizontalAlignment=HAlign_Center}}
  }
  Overlay_2 {
   widgetClassPath=&Overlay
   parent=HorizontalBox_0
   widget=Overlay_2
   slot {class=HorizontalBoxSlot, properties {horizontalAlignment=HAlign_Right, verticalAlignment=VAlign_Center}}
  }
  NextTabActionWrapper {widgetClassPath=&SizeBox, bIsVariable=true, parent=Overlay_2, widget=NextTabActionWrapper, slot {class=OverlaySlot}}
  NextTabAction {
   widgetClassPath=&CommonActionWidget
   bIsVariable=true
   parent=NextTabActionWrapper
   widget=NextTabAction
   slot {class=SizeBoxSlot}
  }
  ButtonRight {
   widgetClassPath=&IconTextButton
   bIsVariable=true
   parent=Overlay_2
   widget=ButtonRight
   properties {
    bIsFocusable=false
    buttonText=""
    clickMethod=MouseDown
    onButtonBaseClicked=()
    onButtonBaseDoubleClicked=()
    onButtonBaseHovered=()
    onButtonBaseUnhovered=()
    onSelectedChangedBase=()
    style=&ButtonStyle-Rotator-Small_C
    visibility=Collapsed
    iconBrush {
     bIsDynamicallyLoaded=false
     drawAs=Image
     imageType=NoImage
     mirroring=NoMirror
     resourceName=None
     resourceObject=&M-Button-Outline-Base-Left
     tiling=NoTile
     imageSize=<12 12>
     margin=<ltrb 0 0 0 0>
     outlineSettings {
      bUseBrushTransparency=false
      roundingType=HalfHeightRadius
      width=0
      color {colorUseRule=UseColor_Specified, specifiedColor=<rgba 0 0 0 0>}
      cornerRadii=<0 0 0 1>
     }
     tintColor {colorUseRule=UseColor_Specified, specifiedColor=<rgba 0 0 0 1>}
     uVRegion {bIsValid=false, max=<0 0>, min=<0 0>}
    }
    pressedSlateSoundOverride {resourceObject=/Game/Sounds/Fort_UI_Sounds/FrontEnd/DefaultSounds/PressedButtons/FN_UI_Default_Button_Pressed_Cue}
    renderTransform {angle=0, scale=<-1 1>, shear=<0 0>, translation=<0 0>}
   }
   slot {class=OverlaySlot, properties {padding=<ltrb 25 0 0 0>}}
  }
  ImageHorizRule {
   widgetClassPath=&Image
   bIsVariable=true
   parent=Overlay_Container
   widget=ImageHorizRule
   properties {
    visibility=Collapsed
    brush {
     bIsDynamicallyLoaded=false
     drawAs=Box
     imageType=NoImage
     mirroring=NoMirror
     resourceName=None
     resourceObject=/Game/UI/Foundation/Textures/Buttons/T-Tab-Secondary-Base-32x64
     tiling=NoTile
     imageSize=<48 88>
     margin=<ltrb 0.49 0.49 0.49 0.49>
     outlineSettings {
      bUseBrushTransparency=false
      roundingType=HalfHeightRadius
      width=0
      color {colorUseRule=UseColor_Specified, specifiedColor=<rgba 0 0 0 0>}
      cornerRadii=<0 0 0 0>
     }
     tintColor {colorUseRule=UseColor_Specified, specifiedColor=<rgba 1 1 1 0.4>}
     uVRegion {bIsValid=false, max=<0 0>, min=<0 0>}
    }
   }
   slot {class=OverlaySlot, properties {horizontalAlignment=HAlign_Fill, verticalAlignment=VAlign_Center, padding=<ltrb 6 0 6 0>}}
  }
 ]
}
