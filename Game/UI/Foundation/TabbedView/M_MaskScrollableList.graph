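# Translucent UI material that, going by its name, masks a scrollable list in tabbed views by fading it at two opposite ends.
# Two MF_UI_LinearGradient calls, one rotated by 0.5, are combined with Min and multiplied by the Texture parameter's alpha.
# The gradient end comes from UV channel 3's G divided by itself and scaled, so it stays a fixed fraction wherever G is non-zero.
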Material /Game/UI/Foundation/TabbedView/M_MaskScrollableList
graph {
 properties {
  bUsedWithStaticMesh=false
  blendMode=BLEND_Translucent
  materialDomain=MD_UI
  pixelDepthOffsetMode=PDOM_Legacy
  usedShadingModels=Default Lit
 }
 nodes [
  n1 = Constant()
  n2 = MaterialFunctionCall(Float2=TextureCoordinate() {coordinateIndex=3}) {materialFunction=&BreakOutFloat2Components}
  n3 = Multiply(A=Divide(A=@n2.G, B=@n2.G), B=Constant() {r=0.4})
  n4 = Constant() {r=0.5}
  Multiply(
   A=TextureSampleParameter2D() {parameterName=Texture}.A
   B=Min(
    A=MaterialFunctionCall(GradientStart=@n1, GradientEnd=@n3) {materialFunction=&MF_UI_LinearGradient}.Result
    B=MaterialFunctionCall(GradientStart=@n1, GradientEnd=@n3, RotationAngle=@n4) {materialFunction=&MF_UI_LinearGradient}.Result
   )
  )
 ]
}
